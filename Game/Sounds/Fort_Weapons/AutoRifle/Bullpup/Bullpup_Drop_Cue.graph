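# Sound cue for dropping the Bullpup auto rifle.
# A random node picks with equal weight among four shared generic assault-rifle drop waves from the gameplay UI sounds.
# Playback uses the weapon drop attenuation, the loot collect sound class and a shared weapon-drop concurrency set, with a slight volume boost.
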SoundCue /Game/Sounds/Fort_Weapons/AutoRifle/Bullpup/Bullpup_Drop_Cue
root=Random_0
nodes [
 {id=Random_0, children [WavePlayer_9, WavePlayer_10, WavePlayer_11, WavePlayer_12], properties {weights [1,1,1,1]}}
 {id=WavePlayer_10, properties {soundWaveAssetPtr=/Game/Sounds/Fort_UI_Sounds/Gameplay/Weapon_Generic_Drop_AR_01}}
 {id=WavePlayer_11, properties {soundWaveAssetPtr=/Game/Sounds/Fort_UI_Sounds/Gameplay/Weapon_Generic_Drop_AR_02}}
 {id=WavePlayer_12, properties {soundWaveAssetPtr=/Game/Sounds/Fort_UI_Sounds/Gameplay/Weapon_Generic_Drop_AR_03}}
 {id=WavePlayer_9, properties {soundWaveAssetPtr=/Game/Sounds/Fort_UI_Sounds/Gameplay/Weapon_Generic_Drop_AR_04}}
]
properties {
 attenuationSettings=&Fort_Weapon_Drop_Att
 maxDistance=3000
 soundClassObject=&Fort_Loot_Collect
 volumeMultiplier=1.25
 concurrencyOverrides {resolutionRule=PreventNew}
 concurrencySet [&Weapon_Drop_StopOldest]
}
