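# First-person firing sound cue for the Bullpup auto rifle, layering a gunshot with a trailing ammo bounce.
# A mixer combines a random pick of three close-fire waves with a much quieter random ammo-bounce wave, which is delayed about half a second and modulated.
# An outer modulator varies pitch and volume of the whole mix.
# The cue uses first-person attenuation, a ranged-fire sound class and a shared first-person gunfire concurrency.
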SoundCue /Game/Sounds/Fort_Weapons/AutoRifle/Bullpup/Bullpup_Fire_1P_Cue
root=Modulator_2
nodes [
 {id=Delay_0, children [Modulator_1], properties {delayMax=0.6, delayMin=0.5}}
 {id=Mixer_2, children [Random_2,Delay_0], properties {inputVolume [1,0.2]}}
 {id=Modulator_1, children [Random_6], properties {pitchMax=1, pitchMin=0.9, volumeMax=0.8, volumeMin=0.7}}
 {id=Modulator_2, children [Mixer_2], properties {pitchMax=1.1, pitchMin=1, volumeMax=1.2, volumeMin=1.1}}
 {id=Random_2, children [WavePlayer_0,WavePlayer_1,WavePlayer_2], properties {bShouldExcludeFromBranchCulling=true, weights [1,1,1]}}
 {id=Random_6, children [WavePlayer_18, WavePlayer_19, WavePlayer_20], properties {bShouldExcludeFromBranchCulling=true, weights [1,1,1]}}
 {id=WavePlayer_0, properties {soundWaveAssetPtr=/Game/Sounds/Fort_Weapons/AutoRifle/Bullpup/NewBurst/Bullpup_Close_Fire_01}}
 {id=WavePlayer_1, properties {soundWaveAssetPtr=/Game/Sounds/Fort_Weapons/AutoRifle/Bullpup/NewBurst/Bullpup_Close_Fire_02}}
 {id=WavePlayer_18, properties {soundWaveAssetPtr=&/Bullpup_AmmoBounce_01}}
 {id=WavePlayer_19, properties {soundWaveAssetPtr=&/Bullpup_AmmoBounce_02}}
 {id=WavePlayer_2, properties {soundWaveAssetPtr=/Game/Sounds/Fort_Weapons/AutoRifle/Bullpup/NewBurst/Bullpup_Close_Fire_03}}
 {id=WavePlayer_20, properties {soundWaveAssetPtr=&/Bullpup_AmmoBounce_03}}
]
properties {
 attenuationSettings=&Fort_Weapon_Projectile_FirstPerson_Athena_Att
 bDebug=true
 maxDistance=2097152
 soundClassObject=&Fort_WeaponRanged_Fire_1P_PSM
 volumeMultiplier=1
 concurrencyOverrides {bLimitToOwner=true, maxCount=3, resolutionRule=StopOldest}
 concurrencySet [&Weapon_Gunfire_FirstPerson_Conc]
}
