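# Sound cue for the magazine-insert foley of the Bullpup auto rifle reload.
# A single wave player feeds Bullpup_Reload_Insert_01 through a mixer, with cue-level pitch and volume slightly lowered.
# It uses the shared weapon reload attenuation and the weapon foley sound class, overrides concurrency, and has its debug flag on.
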SoundCue /Game/Sounds/Fort_Weapons/AutoRifle/Bullpup/Bullpup_Reload_Insert_Cue
root=Mixer_0
nodes [
 {id=Mixer_0, children [WavePlayer_0], properties {inputVolume [1]}}
 {id=WavePlayer_0, properties {soundWaveAssetPtr=&/Bullpup_Reload_Insert_01}}
]
properties {
 attenuationSettings=&Fort_Weapon_Reload_Att
 bDebug=true
 bOverrideConcurrency=true
 maxDistance=2250
 pitchMultiplier=0.9
 soundClassObject=&Fort_WeaponFoley
 volumeMultiplier=0.852
}
